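# Base editor utility widget for tech-art tools, with no UI of its own, that provides save-game persistence via a configurable class and slot name.
# On Construct it loads the slot or creates a new save object, then calls the stub PostLoad hook; Destruct clears a valid reference.
# SaveGame_Save, not called here, requires a prior load, reloads the object, runs the stub PrepareToSave hook, then writes the slot.
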EditorUtilityWidgetBlueprint /Game/Blueprints/Blutilities/ForTechArt/Widgets/TA_Base_EditorWidget
graphs [
 EventGraph {nodes [
   Event Destruct()
   > Is Valid(InputObject=Get saveGame_Object() -> (saveGame_Object:Save Game Object Reference))
   > Set saveGame_Object(saveGame_Object:Save Game Object Reference) -> (Output_Get:Save Game Object Reference)
   Event Construct()
   > SaveGame_Init()
  ]}
 SaveGame_DoesSaveExist {nodes [
   SaveGame_DoesSaveExist() {is=title K2Node_FunctionEntry}
   > n1 = DoesSaveGameExist!(execute, SlotName:String=Get saveGame_SlotName() -> (saveGame_SlotName:String), UserIndex:int=0) -> (then, ReturnValue:bool)
   > Return Node!(execute, exists:bool=@n1)
  ]}
 SaveGame_Init {nodes [
   SaveGame_Init() {is=title K2Node_FunctionEntry}
   > n1 = SaveGame_Load!(execute, self) -> (then, didASaveAlreadyExist:bool, saveGameObject:Save Game Object Reference)
   > SaveGame_PostLoad!(execute, self, saveGameExisted:bool=@n1, saveGame_Object:Save Game Object Reference=@n1.saveGameObject) -> (then, " ":bool)
  ]}
 SaveGame_IsSaveDataValid {nodes [
   SaveGame_IsSaveDataValid() {is=title K2Node_FunctionEntry}
   n1 = Get saveGame_SlotName()
   n2 = Get saveGame_Class() -> (saveGame_Class:Save Game Class Reference)
   > Return Node!(
    execute
    valid:bool=AND Boolean(A="Not Equal Exactly (String)"(A=@n1), B=IsValidClass!(Class:Object Class Reference=@n2) -> (ReturnValue:bool))
   )
  ]}
 SaveGame_Load {nodes [
   SaveGame_Load() {is=title K2Node_FunctionEntry}
   > Set saveGame_hasLoaded(saveGame_hasLoaded:bool=true) -> (Output_Get:bool)
   > n1 = SaveGame_IsSaveDataValid!(execute, self) -> (then, valid:bool) {is=title K2Node_CallFunction}
   > Branch(Condition=@n1)
   > n2 = DoesSaveGameExist(SlotName=Get saveGame_SlotName(), UserIndex=0)
   > n3 = Branch(Condition=@n2)
   > n4 = LoadGameFromSlot!(execute, SlotName:String=Get saveGame_SlotName(), UserIndex:int=0) -> (then, ReturnValue:Save Game Object Reference)
   > n5 = Set saveGame_Object(saveGame_Object=@n4)
   > Return Node!(execute, didASaveAlreadyExist:bool=true, saveGameObject:Save Game Object Reference=@n5)
   n6 = CreateSaveGameObject!(execute=@n3.else, SaveGameClass:Save Game Class Reference=Get saveGame_Class()) -> (then, ReturnValue:Save Game Object Reference)
   > n7 = Set saveGame_Object(saveGame_Object=@n6)
   > Return Node(didASaveAlreadyExist=false, saveGameObject=@n7)
  ]}
 SaveGame_PostLoad {nodes [
   SaveGame_PostLoad!() -> (then, saveGameExisted:bool=~, saveGame_Object:Save Game Object Reference) {is=title K2Node_FunctionEntry}
   > Return Node!(execute, " ":bool=false)
  ]}
 SaveGame_PrepareToSave {nodes [
   SaveGame_PrepareToSave!() -> (then, SaveGameObject:Save Game Object Reference) {is=title K2Node_FunctionEntry}
   > Return Node(" "=false)
  ]}
 SaveGame_Save {nodes [
   SaveGame_Save() {is=title K2Node_FunctionEntry}
   > Branch(Condition=Get saveGame_hasLoaded() -> (saveGame_hasLoaded:bool))
   > n1 = SaveGame_Load() {is=title K2Node_CallFunction}
   > Is Valid(InputObject=@n1.saveGameObject)
   > SaveGame_PrepareToSave!(execute, self, SaveGameObject:Save Game Object Reference=@n1.saveGameObject) -> (then, " ":bool) {is=title K2Node_CallFunction}
   > SaveGameToSlot!(execute, SaveGameObject:Save Game Object Reference=@n1.saveGameObject, SlotName:String=Get saveGame_SlotName(), UserIndex:int=0) -> (then, ReturnValue:bool)
  ]}
]
variables [
 saveGame_Class {type=Class, default=null}
 saveGame_hasLoaded {type=boolean}
 saveGame_Object {type=SaveGame, default=null}
 saveGame_SlotName {type=string, default=""}
]
widgets {info {rootWidgetClass=null, widgetCount=0}, widgets []}
